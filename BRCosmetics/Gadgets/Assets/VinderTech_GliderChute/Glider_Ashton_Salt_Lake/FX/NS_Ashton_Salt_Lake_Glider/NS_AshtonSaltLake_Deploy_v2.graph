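# Niagara deploy effect for the Ashton Salt Lake glider: a mesh portal with lightning and exhaust.
# Mesh_Portal bursts one particle with four mesh renderers that scales open, holds and closes while curves drive its materials.
# Lightning_Strikes places six short-lived meshes radially in the XZ plane using HLSL on UniqueID, with animated material parameters.
# Exhaust_Pulled_Through spawns growing, fading blue sprites at User.SkeletalMesh bones selected by spawn group, after a brief delay.
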NiagaraSystem /BRCosmetics/Gadgets/Assets/VinderTech_GliderChute/Glider_Ashton_Salt_Lake/FX/NS_Ashton_Salt_Lake_Glider/NS_AshtonSaltLake_Deploy_v2
dependencies {usedDataInterfaces [&NiagaraDataInterfaceCurve], usedDynamicInputs [], usedModules [&SystemState]}
emitters {
 Exhaust_Pulled_Through {
  emitterUpdateScript [
   &EmitterState(Life Cycle Mode=::NewEnumerator1, Loop Behavior=::NewEnumerator1, Loop Delay=0.1, Loop Duration=0.3, UseLoopDelay=true)
   &SpawnRate(
    SpawnRate=&FloatFromCurve(FloatCurve=DataInterface() {propertyValues {Curve {keys [<at 0 128 RCIM_Cubic>,<at 0.25 32 RCIM_Cubic>]}}}, CurveIndex=$Emitter.Age)
    Use Spawn Probability=false
   )
   &SpawnBurst_Instantaneous(Spawn Count=32, Use Spawn Probability=false)
  ]
  particleSpawnScript [
   /Niagara/Modules/Spawn/Initialization/InitializeParticle(
    Color=<rgba 0 0.3291 0.8228 2>
    Lifetime=0.35
    Module.Write Color=true
    Module.Write Lifetime=true
    Module.Write Mass=true
    Module.Write Material Random=true
    Module.Write Position=false
    Module.Write Scale=false
    Module.Write SpriteRotation=true
    Module.Write SpriteSize=true
    Position=&SimulationPosition()
    Sprite Rotation=&NormalizedAngleToDegrees(NormalizedAngle=&UniformRangedFloat(Minimum=0, Maximum=1))
    Sprite Size=&Vector2DFromFloat(Value=&UniformRangedFloat(Minimum=100, Maximum=300))
   )
   /Game/Effects/Niagara/Modules/Groups/FN_SetSpawnGroup_v1/FN_SetSpawnGroup(Emitter Bone Count=32)
   &SampleSkeletalMeshSkeleton(
    Mesh:NiagaraDataInterfaceSkeletalMesh=$User.SkeletalMesh
    Sampled Position Coordinate Space:ENiagaraCoordinateSpace=::Local
    Sampled Rotation Coordinate Space:ENiagaraCoordinateSpace=::Local
    Sampled Velocity Coordinate Space:ENiagaraCoordinateSpace=::Local
    Specific Bone Index:NiagaraInt32=$Particles.SpawnGroup
   )
   /Niagara/Modules/Spawn/Location/SkeletalMeshSkeletonLocation(Inherit Velocity From Mesh=0, Module.Inherit Velocity=false, Module.Use Sampled Bone as Ribbon ID=true)
   &AddVelocity(Velocity=<0 10 0>)
   &CameraOffset(Camera Offset Amount=10)
  ]
  particleUpdateScript [
   &ParticleState()
   &SolveForcesAndVelocity(Clamp Velocity=false, Limit Acceleration=false)
   &ScaleColor(Scale Alpha=&Lerp_Float(A=1, B=0, Alpha=$Particles.NormalizedAge), ScaleA=true, ScaleRGB=true)
   &ScaleSpriteSize(Scale Factor=&Vector2DFromFloat(Value=&FloatFromCurve(FloatCurve=DataInterface() {propertyValues {Curve {keys [<at 0 1 RCIM_Cubic>,<at 1 2 RCIM_Cubic>]}}})))
  ]
  rendererClasses [&NiagaraSpriteRendererProperties]
  renderers [{rendererClass=&NiagaraSpriteRendererProperties}]
 }
 Lightning_Strikes {
  emitterUpdateScript [
   &EmitterState(Life Cycle Mode=::NewEnumerator1, Loop Behavior=::NewEnumerator1, Loop Delay=0.1, Loop Duration=1, UseLoopDelay=true)
   &SpawnBurst_Instantaneous(Spawn Count=6, Use Spawn Probability=false)
  ]
  particleSpawnScript [
   &InitializeParticle(
    Lifetime=0.2
    Mesh Scale=&UniformRangedVector(Minimum=<0.1 0.25 0.25>, Maximum=<0.5 0.25 0.25>)
    Module.Write Color=true
    Module.Write Lifetime=true
    Module.Write Mass=true
    Module.Write Material Random=true
    Module.Write Position=false
    Module.Write Scale=true
    Module.Write SpriteRotation=false
    Module.Write SpriteSize=false
    Position=&SimulationPosition()
    Sprite Size=<5 5>
   )
   /SetVariables(Particles.Theta=HlslExpression() {hlslExpression="radians(((Particles.UniqueID / 6.0) * 360.0))"}) {isSetParameters=true}
   /SetVariables(Particles.MeshOrientation=&FindQuatBetweenTwoVectors(
     From Vector=<1 0 0>
     To Vector=&MakeVector(X=HlslExpression() {hlslExpression="sin(Particles.Theta)"}, Y=0, Z=HlslExpression() {hlslExpression="cos(Particles.Theta)"})
    )) {isSetParameters=true}
   /Game/Effects/Niagara/Modules/Position/FN_OffsetPositionByOrientationAxis(Offset Axis:Vector3f=<200 0 0>) {module=OffsetPositionByMeshOrientation}
  ]
  particleUpdateScript [
   &ParticleState()
   &SolveForcesAndVelocity(Clamp Velocity=false, Limit Acceleration=false)
   &DynamicMaterialParameters(
    Index 0 Param 1=&Lerp_Float(A=0, B=1, Alpha=$Particles.NormalizedAge)
    Index 0 Param 2=&FloatFromCurve(FloatCurve=DataInterface() {propertyValues {Curve {keys [
         <at 0 1 RCIM_Cubic>
         {arriveTangent=10, interpMode=RCIM_Cubic, leaveTangent=10, time=0.25, value=1.5}
         {arriveTangent=-3, interpMode=RCIM_Cubic, leaveTangent=-3, time=0.3, value=4}
         {arriveTangent=-5.6429, interpMode=RCIM_Cubic, leaveTangent=-5.6429, time=0.35, value=1.2}
         <at 1 0.05 RCIM_Cubic>
        ]}}})
    Index 0 Param 3=&FloatFromCurve(FloatCurve=DataInterface() {propertyValues {Curve {keys [
         <at 0.1 1 RCIM_Cubic>
         {arriveTangent=-1.2, interpMode=RCIM_Cubic, leaveTangent=-1.2, time=0.3, value=2}
         {arriveTangent=-7, interpMode=RCIM_Cubic, leaveTangent=-7, time=0.35, value=0.7}
         {arriveTangent=3.5385, interpMode=RCIM_Cubic, leaveTangent=3.5385, time=0.55, value=0.25}
         <at 1 3 RCIM_Cubic>
        ]}}})
    Index 0 Param 4=1
    Module.Param0WriteEnabled=true
    Module.Param1WriteEnabled=true
    Module.Param2WriteEnabled=true
    Module.Param3WriteEnabled=true
   )
   &ScaleColor(
    Scale Alpha=&FloatFromCurve(FloatCurve=DataInterface() {propertyValues {Curve {keys [
         <at 0.15 1 RCIM_Cubic>
         {arriveTangent=-0.9036, interpMode=RCIM_Cubic, leaveTangent=-0.9036, time=0.21, value=0.55}
         {arriveTangent=-0.6962, interpMode=RCIM_Cubic, leaveTangent=-0.6962, time=0.98, value=0.25}
         <at 1 0 RCIM_Cubic>
        ]}}})
    Scale RGB=&VectorFromFloat(Value=&FloatFromCurve(FloatCurve=DataInterface() {propertyValues {Curve {keys [
          <at 0 2 RCIM_Cubic>
          {arriveTangent=-3, interpMode=RCIM_Cubic, leaveTangent=-3, time=0.25, value=1}
          {arriveTangent=-0.6667, interpMode=RCIM_Cubic, leaveTangent=-0.6667, time=0.4, value=0.8}
          <at 1 0.5 RCIM_Cubic>
         ]}}}))
    ScaleA=true
    ScaleRGB=true
   )
  ]
  rendererClasses [&NiagaraMeshRendererProperties]
  renderers [{rendererClass=&NiagaraMeshRendererProperties, rendererIndex=0}]
 }
 Mesh_Portal {
  emitterUpdateScript [
   &EmitterState(Life Cycle Mode=::NewEnumerator1, Loop Behavior=::NewEnumerator1, Loop Duration=1, UseLoopDelay=false)
   &SpawnBurst_Instantaneous(Spawn Count=1, Use Spawn Probability=false)
  ]
  particleSpawnScript [
   &InitializeParticle(
    Lifetime=0.767
    Mesh Scale=<1.6 1.6 1.6>
    Module.Write Color=true
    Module.Write Lifetime=true
    Module.Write Mass=true
    Module.Write Material Random=true
    Module.Write Position=false
    Module.Write Scale=true
    Module.Write SpriteRotation=false
    Module.Write SpriteSize=true
    Position=&SimulationPosition()
    Sprite Size=<5 5>
   )
   /SetVariables(Particles.OpeningColor:LinearColor=<rgba 5 5 5 2>) {isSetParameters=true}
  ]
  particleUpdateScript [
   &ParticleState()
   &SolveForcesAndVelocity(Clamp Velocity=false, Limit Acceleration=false)
   &ScaleMeshSize(Scale Factor=&VectorFromFloat(Value=&FloatFromCurve(FloatCurve=DataInterface() {propertyValues {Curve {keys [<at 0 0>,<at 0.1 1>,<at 0.9 1>,<at 1 0>]}}})))
   /SetVariables(Particles.SpaceMeshSize:Vector3f=&Multiply_Vector(
     A=<15 12 12>
     B=&VectorFromFloat(Value=&FloatFromCurve(FloatCurve=DataInterface() {propertyValues {Curve {keys [<at 0 0>,<at 0.1 1>,<at 0.9 1>,<at 1 0>]}}}))
    )) {isSetParameters=true}
   &DynamicMaterialParameters(
    Index 0 Param 1=1
    Index 0 Param 2=&FloatFromCurve(FloatCurve=DataInterface() {propertyValues {Curve {keys [{arriveTangent=8.8358, interpMode=RCIM_Cubic, leaveTangent=8.8358, tangentMode=RCTM_User, time=0, value=0}, <at 0.25 1 RCIM_Cubic>]}}})
    Index 1 Param 1=&FloatFromCurve(FloatCurve=DataInterface() {propertyValues {Curve {keys [
         {arriveTangent=2.256, interpMode=RCIM_Cubic, leaveTangent=2.256, tangentMode=RCTM_User, time=0, value=0.15}
         {arriveTangent=1.0625, interpMode=RCIM_Cubic, leaveTangent=1.0625, time=0.5, value=1}
         {arriveTangent=-1.7, interpMode=RCIM_Cubic, leaveTangent=-1.7, time=0.8, value=1}
         <at 1 0.15 RCIM_Cubic>
        ]}}})
    Module.Param0WriteEnabled=true
    Module.Param0WriteEnabled1=true
    Module.Param1WriteEnabled=true
    Module.Param1WriteEnabled1=false
    Module.Param2WriteEnabled=true
    Module.Param2WriteEnabled1=false
    Module.Param3WriteEnabled=true
    Module.Param3WriteEnabled1=false
    Write Parameter Index 1=true
   )
  ]
  rendererClasses [&NiagaraMeshRendererProperties]
  renderers [
   {rendererClass=&NiagaraMeshRendererProperties, rendererIndex=0}
   {rendererClass=&NiagaraMeshRendererProperties, rendererIndex=1}
   {rendererClass=&NiagaraMeshRendererProperties, rendererIndex=2}
   {rendererClass=&NiagaraMeshRendererProperties, rendererIndex=3}
  ]
 }
}
summary {
 systemName=NS_AshtonSaltLake_Deploy_v2
 emitters [
  {emitterName=Mesh_Portal, rendererClasses [&NiagaraMeshRendererProperties]}
  {emitterName=Exhaust_Pulled_Through, rendererClasses [&NiagaraSpriteRendererProperties]}
  {emitterName=Lightning_Strikes, rendererClasses [&NiagaraMeshRendererProperties]}
 ]
 userVariables [User.SkeletalMesh {
   type=NiagaraDataInterfaceSkeletalMesh
   defaultValue {
    struct=NiagaraExt_VariableValue_DataInterface
    value {dataInterface=&self:NiagaraDataInterfaceSkeletalMesh_0}
   }
  }]
}
