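# Niagara engine effect for the Ashton Salt Lake glider, driven by forward input (User.ForwardDot) plus a decaying timeline curve for a startup flare.
# Engines pins eight persistent mesh particles with lights across four skeletal mesh sockets, modulating stretch and material parameters by that thrust.
# Pop_Sparks throws two-palette GPU sparks at a thrust-scaled rate; Air_Foil_Trails spawns socket ribbons, gated by ForwardDot.
# Front-end modifiers shrink engine meshes and spark spawn shapes.
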NiagaraSystem /BRCosmetics/Gadgets/Assets/VinderTech_GliderChute/Glider_Ashton_Salt_Lake/FX/NS_Ashton_Salt_Lake_Glider/NS_Ashton_SaltLake_Engines_v3
dependencies {
 usedDataInterfaces [
  &NiagaraDataInterfaceArrayFloat3
  &NiagaraDataInterfaceArrayInt32
  &NiagaraDataInterfaceCurve
  &NiagaraDataInterfaceArrayFloat
  &NiagaraDataInterfaceArrayColor
 ]
 usedDynamicInputs []
 usedModules [&SystemState, &Timeline]
}
emitters {
 Air_Foil_Trails {
  emitterUpdateScript [
   &EmitterState()
   &SpawnRate(
    SpawnRate=&Multiply_Float(
     A=120
     B=&MakeCustomFloatFromBool(
      Bool=&SetBoolByFloatComparison(A=$User.ForwardDot, B=0, Comparison Type=::NewEnumerator6)
      True Float=1
      False Float=0
     )
    )
    Use Spawn Probability=false
   )
  ]
  particleSpawnScript [
   &FN_SetSpawnGroup()
   &InitializeRibbon(Color=<rgba 1 1.2 1 0.5>, Lifetime=0.3, Ribbon Width=15, UsePositionOffset=false, Write Material Random=true)
   &SkeletalMeshLocation(
    Bone / Socket Index=&Add_Integer(A=$Particles.SpawnGroup, B=4)
    MaskSpawnGroup=false
    Mesh Sampling Type=::NewEnumerator5
    Position Sampling=::NewEnumerator1
    Sampled Position Offset=<25 0 0>
    Skeletal Mesh=$User.SkeletalMesh
    Socket Sampling Mode=::NewEnumerator5
   )
  ]
  particleUpdateScript [&ParticleState()]
  rendererClasses [&NiagaraRibbonRendererProperties]
  renderers [{rendererClass=&NiagaraRibbonRendererProperties}]
 }
 Engines {
  emitterUpdateScript [
   &EmitterState(Life Cycle Mode=::NewEnumerator1, Loop Behavior=::NewEnumerator1, Loop Duration Mode=::NewEnumerator1, UseLoopDelay=false)
   &SpawnBurst_Instantaneous(Spawn Count=8, Use Spawn Probability=false, UseLoopCountLimit=false)
  ]
  particleSpawnScript [
   &FN_SetSpawnGroup(Number of Spawn Groups=8)
   /SetVariables(Particles.MeshIndex=&MakeCustomIntFromBool(
     Bool=&SetBoolByIntComparison(A=$Particles.SpawnGroup, B=4, Comparison Type=::NewEnumerator2)
     True INT=1
     False INT=0
    )) {isSetParameters=true}
   &InitializeParticle(
    Color=<rgba 0.2195 0.3916 1 1>
    Lifetime=0
    Mesh Scale=&Multiply_VectorByFloat(
     Vector=&SelectVectorFromArray(
      Direct Array Index=$Particles.SpawnGroup
      Vector Selection Array=DataInterface() {propertyValues {FloatData [<0.85 0.6 0.6>, <1 1 1>, <0.85 0.6 0.6>, <1 1 1>, <0.85 0.85 0.85>, <1 1 1>, <0.85 0.85 0.85>, <1 1 1>]}}
     )
     Float=&FN_FrontEndFloatModifier(Gameplay=1, Front End=0.62)
    )
    Mesh Scale Mode=::NewEnumerator4
    UsePositionOffset=false
    Write Material Random=true
   )
  ]
  particleUpdateScript [
   &ParticleState(Kill Particles When Lifetime Has Elapsed=false, Let Infinitely Lived Particles Die When Emitter Deactivates=true)
   &SkeletalMeshLocation(
    Bone / Socket Index=&SelectIntFromArray(Direct Array Index=$Particles.SpawnGroup, Int Selection Array=DataInterface() {propertyValues {IntData [0,0,1,1,2,2,3,3]}})
    MaskSpawnGroup=false
    Mesh Sampling Type=::NewEnumerator5
    Skeletal Mesh=$User.SkeletalMesh
    Socket Sampling Mode=::NewEnumerator5
   )
   &DynamicMaterialParameters(
    Index 0 Param 1=&Lerp_Float(
     A=1
     B=0
     Alpha=&ClampFloat(Float=&Add_Float(
       A=&RemapRange(Input Value To Remap=$User.ForwardDot, Input Min=-1, Input Max=1, Output Min=0, Output Max=1)
       B=&FloatFromCurve(
        FloatCurve=DataInterface() {propertyValues {Curve {keys [
            {interpMode=RCIM_Cubic, leaveTangentWeight=0.3333, tangentMode=RCTM_User, time=0, value=1}
            {arriveTangent=-2.0541, arriveTangentWeight=0.3333, interpMode=RCIM_Cubic, leaveTangent=-2.0541, tangentMode=RCTM_User, time=1, value=0}
           ]}}}
        CurveIndex=$System.Timeline.NormalizedTime
       )
      ))
    )
    Index 0 Param 2=&RandomRangeFloat(Minimum=0, Maximum=1)
    Index 0 Param 3=&Lerp_Float(
     A=0
     B=2
     Alpha=&ClampFloat(Float=&Add_Float(
       A=&RemapRange(Input Value To Remap=$User.ForwardDot, Input Min=-1, Input Max=1, Output Min=0, Output Max=1)
       B=&FloatFromCurve(
        FloatCurve=DataInterface() {propertyValues {Curve {keys [
            {interpMode=RCIM_Cubic, leaveTangentWeight=0.3333, tangentMode=RCTM_User, time=0, value=1}
            {arriveTangent=-2.0541, arriveTangentWeight=0.3333, interpMode=RCIM_Cubic, leaveTangent=-2.0541, tangentMode=RCTM_User, time=1, value=0}
           ]}}}
        CurveIndex=$System.Timeline.NormalizedTime
       )
      ))
    )
    Index 0 Param 4=&SelectFloatFromArray(Direct Array Index=$Particles.SpawnGroup, Float Selection Array=DataInterface() {propertyValues {FloatData [1,0,1,0]}})
    Module.Param0WriteEnabled=true
    Module.Param1WriteEnabled=true
    Module.Param2WriteEnabled=true
    Module.Param3WriteEnabled=true
   )
   &ScaleMeshSize(Scale Factor=&MakeVector(
     X=1
     Y=&Lerp_Float(
      A=0
      B=2
      Alpha=&ClampFloat(Float=&Add_Float(
        A=&RemapRange(Input Value To Remap=$User.ForwardDot, Input Min=-1, Input Max=1, Output Min=0, Output Max=1)
        B=&FloatFromCurve(
         FloatCurve=DataInterface() {propertyValues {Curve {keys [
             {interpMode=RCIM_Cubic, leaveTangentWeight=0.3333, tangentMode=RCTM_User, time=0, value=1}
             {arriveTangent=-2.0541, arriveTangentWeight=0.3333, interpMode=RCIM_Cubic, leaveTangent=-2.0541, tangentMode=RCTM_User, time=1, value=0}
            ]}}}
         CurveIndex=$System.Timeline.NormalizedTime
        )
       ))
     )
     Z=1
    ))
   /SetVariables(
    Particles.Light Color:LinearColor=&MultiplyLinearColor(Linear Color=$Particles.Color, Scale Factor=&MakeVector4_FromFloat(Float=10))
    Particles.Light Position:NiagaraPosition=&Add_Vector(A=Linked() {linkedVariable {name=Particles.Position, type=NiagaraPosition}}, B=<0 0 0>)
    Particles.LightExponent=4
    Particles.LightRadius=75
    Particles.MeshOrientation=/Niagara/DynamicInputs/TypeConversions/MakeCustomQuaternionFromBool(
     Bool:NiagaraBool=&Boolean_AndOperation(
      Bool1=&SetBoolByFloatComparison(A=&MakeFloatFromInt(Int=$Particles.SpawnGroup), B=2, Comparison Type=::NewEnumerator2)
      Bool2=&SetBoolByFloatComparison(A=&MakeFloatFromInt(Int=$Particles.SpawnGroup), B=3, Comparison Type=::NewEnumerator7)
     )
     True Quat:Quat4f=&MultiplyQuaternion(
      Quaternion A=$StackContext.SkeletalMeshLocation.SampledRotation
      Quaternion B=&MakeQuatFromAxisAngle(Axis=<0 1 0>, AngleInDegrees=180)
     )
     False Quat:Quat4f=$StackContext.SkeletalMeshLocation.SampledRotation
    )
   ) {isSetParameters=true}
  ]
  rendererClasses [&NiagaraMeshRendererProperties, &NiagaraLightRendererProperties]
  renderers [
   {rendererClass=&NiagaraMeshRendererProperties, rendererIndex=0}
   {rendererClass=&NiagaraLightRendererProperties, rendererIndex=1}
  ]
 }
 Pop_Sparks {
  simTarget=GPUComputeSim
  emitterUpdateScript [
   &EmitterState()
   &SpawnRate(
    SpawnRate=&Multiply_Float(
     A=150
     B=&Lerp_Float(
      A=0
      B=3
      Alpha=&ClampFloat(Float=&Add_Float(
        A=&ClampFloat(Float=$User.ForwardDot)
        B=&FloatFromCurve(
         FloatCurve=DataInterface() {propertyValues {Curve {keys [
             {interpMode=RCIM_Cubic, leaveTangentWeight=0.3333, tangentMode=RCTM_User, time=0, value=1}
             {arriveTangent=-2.0541, arriveTangentWeight=0.3333, interpMode=RCIM_Cubic, leaveTangent=-2.0541, tangentMode=RCTM_User, time=1, value=0}
            ]}}}
         CurveIndex=$System.Timeline.NormalizedTime
        )
       ))
     )
    )
    Use Spawn Probability=false
   )
  ]
  particleSpawnScript [
   &FN_SetSpawnGroup()
   /SetVariables(Particles.Random=&UniformAOrBFloat(A=0)) {isSetParameters=true}
   &InitializeParticle(
    Color=&SelectLinearColorFromArray(
     Array Sampling Mode=::NewEnumerator1
     Direct Array Index=&MakeIntFromFloat(Float=$Particles.Random)
     Color Selection Array=DataInterface() {propertyValues {ColorData [<rgba 4 2.4045 1.2965 1>, <rgba 1.2965 3.0988 4 1>]}}
     Random Seed=?
    )
    Lifetime Max=0.175
    Lifetime Min=0.1
    Lifetime Mode=::NewEnumerator1
    Sprite Size Max=<1 4>
    Sprite Size Min=<0.5 2>
    Sprite Size Mode=::NewEnumerator2
    UsePositionOffset=false
    Write Material Random=true
   )
   &ShapeLocation(
    Cylinder Height=1
    Cylinder Radius=5
    MaskSpawnGroup=false
    Non Uniform Scale=&FN_FrontEndVector3Modifier(Gameplay=<1 1 1>, Front End=<0.62 0.62 0.62>)
    Override Seed=false
    Rotation Axis=<0 1 0>
    Rotation Mode=::NewEnumerator1
    Shape Origin=/Niagara/DynamicInputs/Helpers/SimulationPosition()
    Shape Primitive=::NewEnumerator1
    Surface Only=false
    UseOwnerScale=true
   )
   &ShapeLocation(
    Box Size=<30 0 1>
    MaskSpawnGroup=false
    Non Uniform Scale=&FN_FrontEndVector3Modifier(Gameplay=<1 1 1>, Front End=<0.62 0.62 0.62>)
    Offset=<0 0 -1>
    Offset Mode=::NewEnumerator0
    Override Seed=false
    Shape Origin=&SimulationPosition()
    Shape Primitive=::NewEnumerator2
    Surface Only=false
    UseOwnerScale=true
   ) {module=ShapeLocation001}
  ]
  particleUpdateScript [
   &ParticleState()
   &SkeletalMeshLocation(
    Bone / Socket Index=$Particles.SpawnGroup
    MaskSpawnGroup=false
    Mesh Sampling Type=::NewEnumerator5
    Position Sampling=::NewEnumerator1
    Sampled Position Offset=&MaskVectorBySpawnGroup(
     Masked Vector=Linked() {linkedVariable {name=StackContext.ShapeLocation001.ShapePosition, type=NiagaraPosition}}
     Passthrough Vector=&MaskVectorBySpawnGroup(
      Masked Vector=Linked() {linkedVariable {name=StackContext.ShapeLocation001.ShapePosition, type=NiagaraPosition}}
      Passthrough Vector=Linked() {linkedVariable {name=StackContext.ShapeLocation.ShapePosition, type=NiagaraPosition}}
      Spawn Group Mask=1
     )
    )
    Skeletal Mesh=$User.SkeletalMesh
    Socket Sampling Mode=::NewEnumerator5
   )
   &LinearForce(Force=&RotateVectorByQuaternion(
     Vector To Rotate=&RandomRangeVector(Minimum=<-1500 3000 -1500>, Maximum=<1500 6000 1500>)
     Quaternion=$StackContext.SkeletalMeshLocation.SampledRotation
    ))
   &AccelerationForce(Acceleration=<0 0 -800>)
   &CurlNoiseForce(Noise Frequency=1, Noise Strength=5000, Pan Noise=true, Pan Noise Field=<0 10 0>)
   &Drag(Drag=&RandomRangeFloat(Minimum=1, Maximum=4), Use Linear Drag=true, Use Rotational Drag=false)
   &SolveForcesAndVelocity(Clamp Velocity=false, Limit Acceleration=false)
   &ScaleSpriteSize(Uniform Curve Sprite Scale=DataInterface() {propertyValues {Curve {keys [
        {arriveTangent=7.1136, interpMode=RCIM_Cubic, leaveTangent=7.1136, tangentMode=RCTM_User, time=0, value=0}
        {arriveTangent=0.5, interpMode=RCIM_Cubic, leaveTangent=0.5, time=0.18, value=1}
        <at 1 0.5 RCIM_Cubic>
       ]}}})
   &ScaleSpriteSizeBySpeed(Max Scale Factor=<1 2>)
   &ScaleColor(
    Scale Alpha=&FloatFromCurve(FloatCurve=DataInterface() {propertyValues {Curve {keys [<at 0.2 1 RCIM_Cubic>,<at 1 0 RCIM_Cubic>]}}})
    Scale RGB=&MakeVectorFromLinearColorRGB(Color=&LerpLinearColors(
      StartColor=&SelectLinearColorFromArray(
       Array Sampling Mode=::NewEnumerator1
       Direct Array Index=&MakeIntFromFloat(Float=$Particles.Random)
       Color Selection Array=DataInterface() {propertyValues {ColorData [<rgba 4 1.3836 0 1>, <rgba 1 1 1 1>]}}
       Random Seed=?
      )
      EndColor=&SelectLinearColorFromArray(
       Array Sampling Mode=::NewEnumerator1
       Direct Array Index=&MakeIntFromFloat(Float=$Particles.Random)
       Color Selection Array=DataInterface() {propertyValues {ColorData [<rgba 1 0.0572 0 1>, <rgba 0 0.6667 1 1>]}}
       Random Seed=?
      )
      LerpFactor=$Particles.NormalizedAge
     ))
    ScaleA=true
    ScaleRGB=true
   )
  ]
  rendererClasses [&NiagaraSpriteRendererProperties]
  renderers [{rendererClass=&NiagaraSpriteRendererProperties}]
 }
}
summary {
 systemName=NS_Ashton_SaltLake_Engines_v3
 emitters [
  {emitterName=Engines, rendererClasses [&NiagaraMeshRendererProperties, &NiagaraLightRendererProperties]}
  {emitterName=Pop_Sparks, simTarget=GPUComputeSim, rendererClasses [&NiagaraSpriteRendererProperties]}
  {emitterName=Air_Foil_Trails, rendererClasses [&NiagaraRibbonRendererProperties]}
 ]
 userVariables [
  User.bIsFrontEnd {type=NiagaraBool}
  User.bIsFrontEndPreview {type=NiagaraBool}
  User.bisFullyDeployed {type=NiagaraBool}
  User.ForwardDot {type=NiagaraFloat, defaultValue=0, description=""}
  User.RightDot {type=NiagaraFloat, defaultValue=0, description=""}
  User.Size {type=NiagaraFloat, defaultValue=0, description=""}
  User.SkeletalMesh {
   type=NiagaraDataInterfaceSkeletalMesh
   defaultValue {
    struct=NiagaraExt_VariableValue_DataInterface
    value {dataInterface=&self:NiagaraDataInterfaceSkeletalMesh_0}
   }
  }
 ]
}
